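# AudioButtonMatrixColumn is a User Widget that holds one column of AudioButtonToggle buttons for an audio button matrix.
# Its only widget is a vertical box, which serves as the button store, with functions to add, clear, count and fetch buttons.
# GetButtons and GetButtonAt cast the box's children to AudioButtonToggle, and GetButtonAt still returns when the cast fails.
# The event graph's events are empty.
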WidgetBlueprint /AudioWidgets/AudioButtonMatrix/AudioButtonMatrixColumn
graphs [
 AddButton {nodes [
   n1 = AddButton!() -> (then, AudioButtonToggle:Audio Button Toggle Object Reference) {is=title K2Node_FunctionEntry}
   > AddChildToVerticalBox(self=Get ButtonVerticalBox() -> (ButtonVerticalBox:Vertical Box Object Reference), Content=@n1)
  ]}
 ClearButtons {nodes [
   ClearButtons() {is=title K2Node_FunctionEntry}
   > ClearChildren(self=Get ButtonVerticalBox())
  ]}
 EventGraph {nodes [
   Event PreConstruct()
   Event Construct()
   Event Tick()
  ]}
 GetButtonAt {nodes [
   n1 = GetButtonAt!() -> (then, Index:int) {is=title K2Node_FunctionEntry}
   > n2 = Cast To AudioButtonToggle!(
    execute
    Object=GetChildAt!(self:Panel Widget Object Reference=Get ButtonVerticalBox(), Index:int=@n1) -> (ReturnValue:Widget Object Reference)
   ) -> (then, CastFailed, AsAudio Button Toggle:Audio Button Toggle Object Reference) {is=title K2Node_DynamicCast}
   Return Node!(execute=[@n2.CastFailed, @n2.then], Button:Audio Button Toggle Object Reference=@n2)
   Get Children() -> (Children:Array of Audio Button Toggle Object References)
  ]}
 GetButtons {nodes [
   GetButtons() {is=title K2Node_FunctionEntry}
   > n1 = For Each Loop(Array:Array of Widget Object References=GetAllChildren(self=Get ButtonVerticalBox())) -> (Array Element:Widget Object Reference)
   Return Node!(execute=@n1.Completed, ChildButtons:Array of Audio Button Toggle Object References=Get Children())
   > n2 = Cast To AudioButtonToggle(Object=@n1)
   > Add(TargetArray:Array of Audio Button Toggle Object References=Get Children(), NewItem:"Audio Button Toggle Object Reference (by ref)"=@n2) {is=title K2Node_CallArrayFunction}
  ]}
 GetNumButtons {nodes [
   GetNumButtons() {is=title K2Node_FunctionEntry}
   > Return Node!(execute, NumButtons:int=GetChildrenCount!(self:Panel Widget Object Reference=Get ButtonVerticalBox()) -> (ReturnValue:int=0))
  ]}
]
variables []
widgets {
 info {inheritedWidgetCount=0, parentClass=&UserWidget, rootWidgetClass=&VerticalBox, widgetCount=1}
 widgets [ButtonVerticalBox {widgetClassPath=&VerticalBox, bIsVariable=true, widget=ButtonVerticalBox}]
}
