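# AudioButtonMatrix is a UMG widget showing a titled grid of AudioButtonToggle buttons with exclusive or multiple selection.
# PreConstruct lays out Values across NumColumns column widgets, rounds the grid's corner buttons, and calls CreateStubButtons to append non-interactive filler.
# Toggle events route through private handlers that deselect others in exclusive mode and broadcast the matching selection-changed dispatcher.
# Construct selects the first button when exclusive.
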WidgetBlueprint /AudioWidgets/AudioButtonMatrix/AudioButtonMatrix
graphs [
 AddAudioButton {nodes [
   n1 = AddAudioButton!() -> (then, Text:Text, ColumnIndex:int, ButtonEnabled:bool=true) {is=title K2Node_FunctionEntry}
   > n2 = GetOrAddVerticalBox!(execute, self, ColumnIndex:int=@n1.ColumnIndex) -> (then, MatrixColumn:Audio Button Matrix Column Object Reference)
   > n3 = Set Column(Column:Audio Button Matrix Column Object Reference=@n2) -> (Output_Get:Audio Button Matrix Column Object Reference)
   > n4 = GetNumButtons!(execute, self:Audio Button Matrix Column Object Reference=@n3) -> (then, NumButtons:int=0)
   > Set RowIndex(RowIndex:int=@n4) -> (Output_Get:int=0)
   > n5 = Construct Object from Class(Class=/AudioWidgets/AudioButtonToggle/AudioButtonToggle.AudioButtonToggle_C, self=@n3) -> (ReturnValue:Audio Button Toggle Object Reference)
   > Set NewButton(NewButton:Audio Button Toggle Object Reference=@n5) -> (Output_Get:Audio Button Toggle Object Reference)
   K2Node_MathExpression_2 = "(ColumnIndex == 0) &&&& (RowIndex == 0)"!(ColumnIndex:int=Get ColumnIndex() -> (ColumnIndex:int=0)) -> (ReturnValue:bool=~) {is=title K2Node_MathExpression, expression=((ColumnIndex == 0) &&&& ({#1} == 0))}
   > Set CurveTopLeft!(
    execute
    CurveTopLeft:bool=@K2Node_MathExpression_2
    self:Audio Button Toggle Object Reference=Get NewButton() -> (NewButton:Audio Button Toggle Object Reference)
   ) -> (then, Output_Get:bool)
   K2Node_MathExpression_3 = "(ColumnIndex == (NumColumns - 1)) &&&& (RowIndex == 0)"!(ColumnIndex:int=Get ColumnIndex() -> (ColumnIndex=0)) -> (ReturnValue:bool=~) {
    is=title K2Node_MathExpression
    expression=((ColumnIndex == ({#2} - 1)) &&&& ({#1} == 0))
   }
   > Set CurveTopRight!(execute, CurveTopRight:bool=@K2Node_MathExpression_3, self:Audio Button Toggle Object Reference=Get NewButton()) -> (then, Output_Get:bool)
   K2Node_MathExpression_4 = "(ColumnIndex == 0) &&&& (RowIndex == (NumRows - 1))"!(ColumnIndex:int=Get ColumnIndex() -> (ColumnIndex=0), NumRows:int=Get Num Rows!() -> (NumRows:int) {is=title K2Node_MacroInstance}) -> (ReturnValue:bool=~) {is=title K2Node_MathExpression, expression=((ColumnIndex == 0) &&&& ({#1} == (NumRows - 1)))}
   > Set CurveBottomLeft!(execute, CurveBottomLeft:bool=@K2Node_MathExpression_4, self:Audio Button Toggle Object Reference=Get NewButton()) -> (then, Output_Get:bool)
   K2Node_MathExpression_1 = "(ColumnIndex == (NumColumns - 1)) &&&& (RowIndex == (NumRows - 1))"!(ColumnIndex:int=Get ColumnIndex() -> (ColumnIndex=0), NumRows:int=Get Num Rows()) -> (ReturnValue:bool=~) {
    is=title K2Node_MathExpression
    expression=((ColumnIndex == ({#2} - 1)) &&&& ({#1} == (NumRows - 1)))
   }
   > Set CurveBottomRight!(execute, CurveBottomRight:bool=@K2Node_MathExpression_1, self:Audio Button Toggle Object Reference=Get NewButton()) -> (then, Output_Get:bool)
   > RefreshCorners(self:Audio Button Toggle Object Reference=Get NewButton())
   > Set TextValue!(execute, TextValue:Text=Get Text!() -> (Text), self:Audio Button Toggle Object Reference=Get NewButton()) -> (then, Output_Get:Text)
   > AddButton!(
    execute
    self:Audio Button Matrix Column Object Reference=Get Column() -> (Column:Audio Button Matrix Column Object Reference)
    AudioButtonToggle:Audio Button Toggle Object Reference=Get NewButton()
   ) -> (then)
   > n6 = Branch(Condition=Get ButtonEnabled() -> (ButtonEnabled:bool))
   Return Node!(execute=[@n7.then, @n6.then], Button:Audio Button Toggle Object Reference=Get NewButton())
   n7 = SetVisibility(execute=@n6.else, self=Get NewButton(), InVisibility=HitTestInvisible)
  ]}
 MathExpression_2_2_2 {
  path=AddAudioButton.K2Node_MathExpression_1.MathExpression_2_2_2
  nodes [
   n1 = int - int(A=Get NumColumns() -> (NumColumns:int=0), B=1) {is=title K2Node_CallFunction}
   n2 = Inputs!() -> (ColumnIndex:int, NumRows:int)
   n3 = int - int(A=@n2.NumRows, B=1)
   n4 = Get RowIndex() -> (RowIndex:int=0)
   Outputs!(ReturnValue:bool=AND Boolean!(A:bool="Equal (Integer)"(A=@n2, B=@n1) {is=title K2Node_CallFunction}, B:bool="Equal (Integer)"(A=@n4, B=@n3)) -> (ReturnValue:bool) {is=title K2Node_CallFunction})
  ]
 }
 Math Expression {
  path=AddAudioButton.K2Node_MathExpression_2.Math Expression
  nodes [
   n1 = Inputs!() -> (ColumnIndex)
   n2 = Get RowIndex() -> (RowIndex=0)
   Outputs(ReturnValue=AND Boolean(A="Equal (Integer)"(A=@n1, B=0), B="Equal (Integer)"(A=@n2, B=0)))
  ]
 }
 MathExpression_2 {
  path=AddAudioButton.K2Node_MathExpression_3.MathExpression_2
  nodes [
   n1 = int - int(A=Get NumColumns() -> (NumColumns=0), B=1)
   n2 = Inputs()
   n3 = Get RowIndex() -> (RowIndex=0)
   Outputs(ReturnValue=AND Boolean(A="Equal (Integer)"(A=@n2, B=@n1), B="Equal (Integer)"(A=@n3, B=0)))
  ]
 }
 MathExpression_2_2 {
  path=AddAudioButton.K2Node_MathExpression_4.MathExpression_2_2
  nodes [
   n1 = Inputs!() -> (ColumnIndex, NumRows:int)
   n2 = int - int(A=@n1.NumRows, B=1)
   n3 = Get RowIndex() -> (RowIndex=0)
   Outputs(ReturnValue=AND Boolean(A="Equal (Integer)"(A=@n1, B=0), B="Equal (Integer)"(A=@n3, B=@n2)))
  ]
 }
 CreateStubButtons {nodes [
   CreateStubButtons() {is=title K2Node_FunctionEntry}
   K2Node_MathExpression_0 = NumColumns - 1!() -> (ReturnValue:int) {is=title K2Node_MathExpression, expression=({#2} - 1)}
   K2Node_MathExpression_1 = "NumColumns - ((NumColumns * NumRows) % ValueLen)"!(ValueLen:int=Length(TargetArray:Array of Texts=Get Values() -> (Values:Array of Texts)), NumRows:int=Get Num Rows()) -> (ReturnValue:int) {
    is=title K2Node_MathExpression
    expression=({#2} - (({#2} * NumRows) % ValueLen))
   }
   > n1 = For Loop(FirstIndex=@K2Node_MathExpression_1, LastIndex=@K2Node_MathExpression_0)
   > AddAudioButton!(execute, self, Text:Text, ColumnIndex:int=@n1, ButtonEnabled:bool=false) -> (then, Button:Audio Button Toggle Object Reference) {is=title K2Node_CallFunction}
  ]}
 MathExpression_2 {
  path=CreateStubButtons.K2Node_MathExpression_0.MathExpression_2
  nodes [
   Inputs!()
   Outputs(ReturnValue:int=int - int(A=Get NumColumns() -> (NumColumns=0), B=1))
  ]
 }
 MathExpression {
  path=CreateStubButtons.K2Node_MathExpression_1.MathExpression
  nodes [
   n1 = Inputs!() -> (ValueLen:int, NumRows:int)
   n2 = int * int!(A:int=Get NumColumns() -> (NumColumns=0), B:int=@n1.NumRows) -> (ReturnValue:int=0) {is=title K2Node_CallFunction}
   Outputs(ReturnValue=int - int(A=Get NumColumns() -> (NumColumns=0), B="% (Integer)"(A=@n2, B=@n1) -> (ReturnValue=0)))
  ]
 }
 EventGraph {nodes [
   n1 = SetButtonSelected_EventPrivate!() -> (OutputDelegate, then, Button:Audio Button Toggle Object Reference) {is=title K2Node_CustomEvent}
   > SetButtonSelectedPrivate!(execute, self, Button:Audio Button Toggle Object Reference=@n1.Button) -> (then)
   n2 = SetButtonDeselected_EventPrivate!() -> (OutputDelegate, then, Button:Audio Button Toggle Object Reference) {is=title K2Node_CustomEvent}
   > n3 = Branch(Condition=Get ExclusiveSelection() -> (ExclusiveSelection:bool))
   > Branch(Condition="Equal (Object)"(A=Get LastExclusive() -> (LastExclusive:Audio Button Toggle Object Reference), B=@n2.Button) {is=title K2Node_CallFunction})
   > SetSelected!(execute, self:Audio Button Toggle Object Reference=@n2.Button, ReportEvent:bool=false) -> (then)
   SetSelectionChangedPrivate!(execute=@n3.else, self, NewSelection:Audio Button Toggle Object Reference=@n2.Button, ReportDeselected:bool=false) -> (then)
   Event PreConstruct()
   > "SetText (Text)"(self=Get TitleText() -> (TitleText:Text Object Reference), InText=Get DisplayName() -> (DisplayName:Text))
   > SetToolTipText!(execute, self:Widget Object Reference=Get TitleText(), InToolTipText:"Text (by ref)"=Get Tooltip() -> (Tooltip:Text)) -> (then)
   > ClearChildren(self=Get HorizontalButtonBox() -> (HorizontalButtonBox:Horizontal Box Object Reference))
   > n4 = Branch(Condition="integer > integer"(A=Get NumColumns() -> (NumColumns=0), B=0) {is=title K2Node_CallFunction})
   n5 = For Each Loop(Exec=[@n4.then, @n7.then], Array:Array of Texts=Get Values()) -> (Array Element:Text)
   CreateStubButtons(execute=@n5.Completed) {is=title K2Node_CallFunction}
   n6 = AddAudioButton(
    execute=@n5.LoopBody
    Text=@n5
    ColumnIndex="% (Integer)"(A=@n5.Array Index, B=Get NumColumns() -> (NumColumns=0)) -> (ReturnValue=0)
    ButtonEnabled=true
   )
   > Bind Event to On Selected(self:Audio Button Toggle Object Reference=@n6, Delegate=@n1) {is=title K2Node_AddDelegate}
   > Bind Event to On Deselected(self:Audio Button Toggle Object Reference=@n6, Delegate=@n2) {is=title K2Node_AddDelegate}
   n7 = Set NumColumns(execute=@n4.else, NumColumns:int=Length(TargetArray=Get Values())) -> (Output_Get:int=0)
   Event Construct()
   n8 = Get HorizontalButtonBox()
   > Branch(Condition=AND Boolean(A=HasAnyChildren!(self:Panel Widget Object Reference=@n8) -> (ReturnValue:bool), B=Get ExclusiveSelection()) {is=title K2Node_CommutativeAssociativeBinaryOperator})
   > SetSelected!(execute, self:Self Object Reference, Row:int=0, Column:int=0) -> (then)
  ]}
 GetNumRows {nodes [
   Inputs!()
   n1 = Length(TargetArray=Get Values())
   n2 = Get NumColumns() -> (NumColumns=0)
   n3 = "integer > integer"(A="% (Integer)"(A=@n1, B=@n2) -> (ReturnValue=0), B=0)
   Outputs!(NumRows:int=int + int(
     A=int / int!(A:int=@n1, B:int=@n2) -> (ReturnValue:int=0) {is=title K2Node_CallFunction}
     B="To Integer (Boolean)"!(InBool:bool=@n3) -> (ReturnValue:int=0)
    ) -> (ReturnValue=0) {is=title K2Node_CommutativeAssociativeBinaryOperator})
  ]}
 GetOrAddVerticalBox {nodes [
   n1 = GetOrAddVerticalBox!() -> (then, ColumnIndex:int) {is=title K2Node_FunctionEntry}
   n2 = Get HorizontalButtonBox()
   > n3 = Branch(Condition="integer > integer"(A=GetChildrenCount!(self:Panel Widget Object Reference=@n2) -> (ReturnValue:int=0), B=@n1))
   n4 = Cast To AudioButtonMatrixColumn!(
    execute=[@n3.then, @n6.Completed]
    Object=GetChildAt!(self:Panel Widget Object Reference=Get HorizontalButtonBox(), Index:int=Get ColumnIndex() -> (ColumnIndex=0)) -> (ReturnValue:Widget Object Reference)
   ) -> (then, CastFailed, AsAudio Button Matrix Column:Audio Button Matrix Column Object Reference) {is=title K2Node_DynamicCast}
   Return Node!(execute=[@n4.CastFailed, @n4.then], MatrixColumn:Audio Button Matrix Column Object Reference=@n4)
   n5 = Get IndexValid() -> (IndexValid:bool)
   n6 = While Loop!(
    execute:Exec=@n3.else
    Condition:bool=AND Boolean(A=NOT Boolean(A=@n5), B="integer >= integer"!(A:int=@n1, B:int=0) -> (ReturnValue:bool) {is=title K2Node_CallFunction})
   ) -> (LoopBody:Exec, Completed:Exec) {is=title K2Node_MacroInstance}
   > n7 = Construct Object from Class(Class=&/AudioButtonMatrixColumn.AudioButtonMatrixColumn_C, self=Get HorizontalButtonBox()) -> (ReturnValue:Audio Button Matrix Column Object Reference)
   > AddChildToHorizontalBox!(execute, self:Horizontal Box Object Reference=Get HorizontalButtonBox(), Content:Widget Object Reference=@n7) -> (then, ReturnValue:Horizontal Box Slot Object Reference)
   n8 = Get HorizontalButtonBox()
   > Set IndexValid(IndexValid:bool="integer > integer"(A=GetChildrenCount(self=@n8) -> (ReturnValue=0), B=Get ColumnIndex() -> (ColumnIndex=0))) -> (Output_Get:bool)
  ]}
 OnExclusiveSelectionChanged {nodes [OnExclusiveSelectionChanged!() -> (then, Selection:Audio Button Toggle Object Reference) {is=title K2Node_FunctionEntry}]}
 OnMultiSelectionChanged {nodes [OnMultiSelectionChanged!() -> (then, SelectedButtons:Array of Audio Button Toggle Object References) {is=title K2Node_FunctionEntry}]}
 SetButtonSelectedPrivate {nodes [
   n1 = SetButtonSelectedPrivate!() -> (then, Button:Audio Button Toggle Object Reference) {is=title K2Node_FunctionEntry}
   > n2 = Branch(Condition=Get ExclusiveSelection())
   > Set LastExclusive(LastExclusive:Audio Button Toggle Object Reference=Get Button() -> (Button:Audio Button Toggle Object Reference)) -> (Output_Get:Audio Button Toggle Object Reference)
   > n3 = Call On Exclusive Selection Changed!(execute:Exec, self:Audio Button Matrix Object Reference, Selection:Audio Button Toggle Object Reference=@n1) -> (then:Exec) {is=title K2Node_CallDelegate}
   SetSelectionChangedPrivate(execute=[@n3.then, @n2.else], NewSelection=@n1, ReportDeselected=true)
  ]}
 SetSelected {nodes [
   n1 = SetSelected!() -> (then, Row:int, Column:int) {is=title K2Node_FunctionEntry}
   > n2 = Cast To AudioButtonMatrixColumn(Object=GetChildAt(self=Get HorizontalButtonBox(), Index=@n1.Column))
   > n3 = GetButtonAt!(execute, self:Audio Button Matrix Column Object Reference=@n2, Index:int=@n1) -> (then, Button:Audio Button Toggle Object Reference)
   > Is Valid(InputObject=@n3)
   > SetSelected!(execute, self:Audio Button Toggle Object Reference=@n3, ReportEvent:bool=true) -> (then) {is=title K2Node_CallFunction}
  ]}
 SetSelectedByArrayIndex {nodes [
   n1 = SetSelectedByArrayIndex!() -> (then, Index:int) {is=title K2Node_FunctionEntry}
   > Branch(Condition="integer > integer"(A=@n1, B=0))
   > SetSelected!(
    execute
    self:Self Object Reference
    Row:int=int / int(A=@n1, B=Get NumColumns() -> (NumColumns=0)) -> (ReturnValue=0)
    Column:int="% (Integer)"(A=@n1, B=Get Num Rows()) -> (ReturnValue=0)
   ) -> (then)
  ]}
 SetSelectionChangedPrivate {nodes [
   SetSelectionChangedPrivate!() -> (then, NewSelection:Audio Button Toggle Object Reference, ReportDeselected:bool=~) {is=title K2Node_FunctionEntry}
   > n1 = For Each Loop(Array:Array of Widget Object References=GetAllChildren(self=Get HorizontalButtonBox())) -> (Array Element:Widget Object Reference)
   n2 = Branch(execute=@n1.Completed, Condition=Get ExclusiveSelection())
   Call On Multi Selection Changed!(
    execute:Exec=@n2.else
    self:Audio Button Matrix Object Reference
    SelectedButtons:Array of Audio Button Toggle Object References=Get SelectedButtons() -> (SelectedButtons:Array of Audio Button Toggle Object References)
   ) -> (then:Exec) {is=title K2Node_CallDelegate}
   n3 = Cast To AudioButtonMatrixColumn(execute=@n1.LoopBody, Object=@n1)
   > n4 = GetButtons!(execute, self:Audio Button Matrix Column Object Reference=@n3) -> (then, ChildButtons:Array of Audio Button Toggle Object References)
   > n5 = For Each Loop(Array:Array of Audio Button Toggle Object References=@n4) -> (Array Element:Audio Button Toggle Object Reference)
   > n6 = GetIsSelected!(execute, self:Audio Button Toggle Object Reference=@n5) -> (then, IsSelected:bool)
   > n7 = Branch(Condition=@n6)
   > n8 = Add Unique(
    TargetArray:Array of Audio Button Toggle Object References=Get SelectedButtons()
    NewItem:"Audio Button Toggle Object Reference (by ref)"=@n5
   )
   n9 = Get NewSelection() -> (NewSelection:Audio Button Toggle Object Reference)
   Branch(
    execute=[@n8.then, @n7.else]
    Condition=AND Boolean!(
     A="Not Equal (Object)"!(A:Object Reference=@n5, B:Object Reference=@n9) -> (ReturnValue:bool)
     B=Get ReportDeselected() -> (ReportDeselected:bool)
     C:bool=Get ExclusiveSelection()
    ) -> (ReturnValue)
   )
   > SetDeselected(self:Audio Button Toggle Object Reference=@n5)
  ]}
]
variables [
 DisplayName {type=string, default=Control01}
 ExclusiveSelection {type=boolean}
 LastExclusive {type=AudioButtonToggle_C, default=null}
 NumColumns {type=integer}
 OnExclusiveSelectionChanged {}
 OnMultiSelectionChanged {}
 Tooltip {type=string, default=""}
 Values {type="array<string>", default []}
]
widgets {
 info {inheritedWidgetCount=0, parentClass=&UserWidget, rootWidgetClass=&CanvasPanel, widgetCount=4}
 widgets [
  ContentPanel {widgetClassPath=&CanvasPanel, widget=ContentPanel}
  HorizonalContentBox {
   widgetClassPath=&HorizontalBox
   parent=ContentPanel
   widget=HorizonalContentBox
   slot {class=CanvasPanelSlot, properties {layoutData {alignment=<0 0>, anchors {maximum=<1 1>, minimum=<0 0>}, offsets=<ltrb 0 0 0 0>}}}
  }
  TitleText {
   widgetClassPath=&TextBlock
   bIsVariable=true
   parent=HorizonalContentBox
   widget=TitleText
   properties {
    text=Control01
    colorAndOpacity {colorUseRule=UseColor_Specified, specifiedColor=<rgba 0.3813 0.3813 0.3813 1>}
    font {
     bForceMonospaced=false
     bMaterialIsStencil=false
     fontMaterial=null
     fontObject=&Roboto
     letterSpacing=0
     monospacedWidth=1
     size=10
     skewAmount=0
     typefaceFontName=Regular
     outlineSettings {
      bApplyOutlineToDropShadows=false
      bMiteredCorners=false
      bSeparateFillAlpha=false
      outlineMaterial=null
      outlineSize=0
      outlineColor=<rgba 0 0 0 1>
     }
    }
   }
   slot {class=HorizontalBoxSlot, properties {verticalAlignment=VAlign_Top, padding=<ltrb 0 4 4 0>}}
  }
  HorizontalButtonBox {
   widgetClassPath=&HorizontalBox
   bIsVariable=true
   parent=HorizonalContentBox
   widget=HorizontalButtonBox
   slot {class=HorizontalBoxSlot, properties {size {sizeRule=Fill, value=1}}}
  }
 ]
}
